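# Material function for the Athena landscape's arid sand layer, outputting a packed Simple_Material plus separate color, spec, height and normal outputs.
# Input UVs drive color, mask and normal samples; color is tinted by the landscape parameter collection, with an untinted output too.
# Color green drives specular, and color red serves as height, packed into Opacity.
# UseRoughness, true by default, selects constant roughness 1; false uses a mask-blue-driven range.
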MaterialFunction /FigmentEnvironment/FN_6_20/Athena/Environments/Landscape/MaterialFunctions/Arid/MF_Athena_Arid_Sand
graph {nodes [
  Constant3Vector() {constant=<rgba 0 0 1 0>}
  n1 = FunctionInput(Preview=TextureCoordinate()) {bUsePreviewValueAsDefault=true, inputName=UVs, inputType=FunctionInput_Vector2}
  n2 = TextureSampleParameter2D(UVs=@n1) {parameterName=Arid_Sand_01, texture=/FigmentEnvironment/FN_6_20/Environments/Sets/Arid/Textures/T_desert_sand_base_C}
  n3 = Multiply(
   A=@n2.RGB
   B=ComponentMask(None=CollectionParameter() {collection=/FigmentEnvironment/FN_6_20/Athena/Environments/Landscape/MPC/MPC_Landscape, parameterName=Arid_Sand_Tint}) {b=true, g=true, r=true}
  )
  n4 = LinearInterpolate(
   A=ScalarParameter() {group=Arid_Sand, parameterName=AridSand_SpecMin}
   B=ScalarParameter() {defaultValue=0.75, group=Arid_Sand, parameterName=AridSand_SpecMax}
   Alpha=@n2.G
  )
  n5 = Constant() {r=1}
  n6 = LinearInterpolate(
   A=ScalarParameter() {defaultValue=-0.5, group=Arid_Sand, parameterName=Sand_Rough_Min}
   B=ScalarParameter() {defaultValue=1.2, group=Arid_Sand, parameterName=Sand_Rough_Max}
   Alpha=TextureSample(UVs=@n1) {samplerType=SAMPLERTYPE_Masks, texture=/FigmentEnvironment/FN_6_20/Environments/Sets/Arid/Textures/T_desert_sand_base_S}.B
  )
  n7 = FunctionInput(Preview=StaticBool() {value=true}) {bUsePreviewValueAsDefault=true, inputName=UseRoughness, inputType=FunctionInput_StaticBool}
  n8 = TextureSampleParameter2D(UVs=@n1) {
   parameterName=Arid_Sand_01_Normal
   samplerType=SAMPLERTYPE_Normal
   texture=/FigmentEnvironment/FN_6_20/Environments/Sets/Arid/Textures/T_desert_sand_base_N
  }
  FunctionOutput(None=MakeMaterialAttributes!(
    BaseColor=FunctionOutput(None=@n3) {outputName=BaseColor}
    Metallic
    Specular=FunctionOutput(None=@n4) {outputName=Spec}
    Roughness=StaticSwitch(True=@n5, False=@n6, Value=@n7)
    Anisotropy
    EmissiveColor
    Opacity=FunctionOutput(None=@n2.R) {outputName=Height}
    OpacityMask
    Normal=FunctionOutput(None=@n8.RGB) {outputName=Normal}
    Tangent
    WorldPositionOffset
    SubsurfaceColor
    ClearCoat
    ClearCoatRoughness
    AmbientOcclusion
    Refraction
    CustomizedUVs_0
    CustomizedUVs_1
    CustomizedUVs_2
    CustomizedUVs_3
    CustomizedUVs_4
    CustomizedUVs_5
    CustomizedUVs_6
    CustomizedUVs_7
    PixelDepthOffset
    ShadingModel
    Displacement
   )) {outputName=Simple_Material}
  FunctionOutput(None=@n2.RGB) {outputName=BaseColor_Plain}
 ]}
